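# Sound cue for opening a toolbox container, part of the loot foley set.
# A Random node picks one of four equally weighted open recordings and feeds a Modulator that pins volume and leaves default pitch variation.
# It uses mono item-interact attenuation and the loot reward sound class.
# A concurrency override caps simultaneous plays, stopping the farthest before refusing new ones.
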SoundCue /Game/Sounds/Foley_Loot/Containers/Toolbox/Toolbox_Open_Cue
root=Modulator_0
nodes [
 {id=Modulator_0, children [Random_0], properties {volumeMax=1, volumeMin=1}}
 {
  id=Random_0
  children [WavePlayer_0,WavePlayer_1,WavePlayer_2,WavePlayer_3]
  properties {bShouldExcludeFromBranchCulling=true, weights [1,1,1,1]}
 }
 {id=WavePlayer_0, properties {soundWaveAssetPtr=&/Toolbox_Open_01}}
 {id=WavePlayer_1, properties {soundWaveAssetPtr=&/Toolbox_Open_02}}
 {id=WavePlayer_2, properties {soundWaveAssetPtr=&/Toolbox_Open_03}}
 {id=WavePlayer_3, properties {soundWaveAssetPtr=&/Toolbox_Open_04}}
]
properties {
 attenuationSettings=/Game/Sounds/Attenuation/Fort_Item_Interact_Mono_Att
 bOverrideConcurrency=true
 maxDistance=1500
 soundClassObject=&Fort_Loot_Reward_PSM
 volumeMultiplier=1
 concurrencyOverrides {maxCount=10, resolutionRule=StopFarthestThenPreventNew}
}
