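# Sound cue for the end of searching a toolbox loot container, though its three waves are named as loop-start recordings.
# A Random node picks one of the equally weighted waves and feeds a Modulator that varies pitch on each play.
# The cue also raises base pitch, lowers volume, uses item-interact mono attenuation and the loot-collect sound class.
# A concurrency override caps simultaneous instances, stopping the farthest first.
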SoundCue /Game/Sounds/Foley_Loot/Containers/Toolbox/Toolbox_Searching_End_Cue
root=Modulator_0
nodes [
 {id=Modulator_0, children [Random_0], properties {pitchMax=1.2, pitchMin=0.9}}
 {id=Random_0, children [WavePlayer_0,WavePlayer_1,WavePlayer_2], properties {bShouldExcludeFromBranchCulling=true, weights [1,1,1]}}
 {id=WavePlayer_0, properties {soundWaveAssetPtr=&/Toolbox_Searching_Loop_Start_01}}
 {id=WavePlayer_1, properties {soundWaveAssetPtr=&/Toolbox_Searching_Loop_Start_02}}
 {id=WavePlayer_2, properties {soundWaveAssetPtr=&/Toolbox_Searching_Loop_Start_03}}
]
properties {
 attenuationSettings=/Game/Sounds/Attenuation/Fort_Item_Interact_Mono_Att
 bOverrideConcurrency=true
 maxDistance=1500
 pitchMultiplier=1.2
 soundClassObject=&Fort_Loot_Collect
 volumeMultiplier=0.5
 concurrencyOverrides {maxCount=4, resolutionRule=StopFarthestThenPreventNew}
}
