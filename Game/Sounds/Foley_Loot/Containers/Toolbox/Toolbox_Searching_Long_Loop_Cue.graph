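# Sound cue for the long toolbox searching loop, a loot container foley sound using item-interact attenuation and the loot-collect sound class.
# A mixer layers two modulated branches that begin together: a random pick of three one-shot start sounds and a random pick of two looping search beds.
# The start branch gets a wider pitch variation, while the loop branch holds volume fixed.
# Overridden concurrency prevents new instances once its limit is reached.
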SoundCue /Game/Sounds/Foley_Loot/Containers/Toolbox/Toolbox_Searching_Long_Loop_Cue
root=Mixer_3
nodes [
 {id=Mixer_3, children [Modulator_3,Modulator_4], properties {inputVolume [1,1]}}
 {id=Modulator_3, children [Random_1], properties {pitchMax=1.2, pitchMin=0.9}}
 {id=Modulator_4, children [Random_2], properties {volumeMax=1, volumeMin=1}}
 {id=Random_1, children [WavePlayer_2,WavePlayer_4,WavePlayer_6], properties {bShouldExcludeFromBranchCulling=true, weights [1,1,1]}}
 {id=Random_2, children [WavePlayer_7,WavePlayer_8], properties {bShouldExcludeFromBranchCulling=true, weights [1,1]}}
 {id=WavePlayer_2, properties {soundWaveAssetPtr=&/Toolbox_Searching_Loop_Start_01}}
 {id=WavePlayer_4, properties {soundWaveAssetPtr=&/Toolbox_Searching_Loop_Start_02}}
 {id=WavePlayer_6, properties {soundWaveAssetPtr=&/Toolbox_Searching_Loop_Start_03}}
 {id=WavePlayer_7, properties {bLooping=true, soundWaveAssetPtr=&/Toolbox_Searching_Loop_01}}
 {id=WavePlayer_8, properties {bLooping=true, soundWaveAssetPtr=&/Toolbox_Searching_Loop_02}}
]
properties {
 attenuationSettings=/Game/Sounds/Attenuation/Fort_Item_Interact_Mono_Att
 bOverrideConcurrency=true
 maxDistance=1500
 soundClassObject=&Fort_Loot_Collect
 volumeMultiplier=1.406
 concurrencyOverrides {resolutionRule=PreventNew}
}
